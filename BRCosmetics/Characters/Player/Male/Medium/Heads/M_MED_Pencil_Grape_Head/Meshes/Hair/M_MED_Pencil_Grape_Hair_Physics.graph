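# Physics asset for the hair of the male medium Pencil Grape head cosmetic, providing bodies for secondary hair motion.
# Each dyn_hair bone across the front, mid and back groups, including two Grape-specific front bones, gets a small capsule, while one large capsule covers the head.
# Every hair body is constrained directly to the head rather than chained, with uniform tight swing limits and a tighter twist limit.
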PhysicsAsset /BRCosmetics/Characters/Player/Male/Medium/Heads/M_MED_Pencil_Grape_Head/Meshes/Hair/M_MED_Pencil_Grape_Hair_Physics
bodies {
 dyn_hairBack_00 [{
   shapeType=Capsule
   length=2
   radius=1
   shapeName=dyn_hairFront_00_capsule
   center=<1.0398 0 -0.0229>
   rotation=<rot 80.4652 -179.9998 -179.9998>
  }]
 dyn_hairBack_01 [{
   shapeType=Capsule
   length=2
   radius=1
   shapeName=dyn_hairFront_00_capsule
   center=<1.0398 0 -0.0229>
   rotation=<rot 80.4652 -179.9998 -179.9998>
  }]
 dyn_hairBack_02 [{
   shapeType=Capsule
   length=3
   radius=1
   shapeName=dyn_hairFront_00_capsule
   center=<1.3931 0 0.0365>
   rotation=<rot 80.4652 -179.9998 -179.9998>
  }]
 dyn_hairFront_00 [{shapeType=Capsule, length=3, radius=1, shapeName=dyn_hairFront_00_capsule, center=<1.5 0 -0.6>, rotation=<rot 75 0 0>}]
 dyn_hairFront_01 [{shapeType=Capsule, length=4, radius=1, shapeName=dyn_hairFront_01_capsule, center=<1.1719 0 -0.6738>, rotation=<rot 69.9634 -0.0002 0.0002>}]
 dyn_hairFront_02 [{
   shapeType=Capsule
   length=3.25
   radius=1
   shapeName=dyn_hairFront_02_capsule
   center=<1.3 0 -0.1625>
   rotation=<rot -82.6214 179.9995 -179.9996>
  }]
 dyn_hairFront_03_Grape [{shapeType=Capsule, length=4, radius=1, shapeName=dyn_hairMid_04_capsule, center=<2 -0.087 0.155>, rotation=<rot -84.4369 0.0011 -3.1173>}]
 dyn_hairFront_04_Grape [{shapeType=Capsule, length=4, radius=1, shapeName=dyn_hairMid_04_capsule, center=<2 -0.087 0.155>, rotation=<rot -84.4369 0.0011 -3.1173>}]
 dyn_hairMidBack_00 [{
   shapeType=Capsule
   length=3
   radius=1
   shapeName=dyn_hairFront_00_capsule
   center=<1.1908 0 0.2953>
   rotation=<rot 80.1602 -179.9998 -179.9998>
  }]
 dyn_hairMidBack_01 [{
   shapeType=Capsule
   length=3
   radius=1
   shapeName=dyn_hairFront_00_capsule
   center=<1.1908 0 0.2953>
   rotation=<rot 80.1602 -179.9998 -179.9998>
  }]
 dyn_hairMidBack_02 [{shapeType=Capsule, length=4, radius=1, shapeName=dyn_hairMid_04_capsule, center=<2 -0.087 0.155>, rotation=<rot -84.4369 0.0011 -3.1173>}]
 dyn_hairMidBack_03 [{
   shapeType=Capsule
   length=3
   radius=1
   shapeName=dyn_hairFront_00_capsule
   center=<1.1908 0 0.2953>
   rotation=<rot 80.1602 -179.9998 -179.9998>
  }]
 dyn_hairMidFront_00 [{shapeType=Capsule, length=3, radius=1, shapeName=dyn_hairFront_00_capsule, center=<1.5 0 -0.6>, rotation=<rot 75 0 0>}]
 dyn_hairMidFront_01 [{shapeType=Capsule, length=4, radius=1, shapeName=dyn_hairMid_04_capsule, center=<2 -0.087 0.155>, rotation=<rot -84.4369 0.0011 -3.1173>}]
 dyn_hairMidFront_02 [{shapeType=Capsule, length=2, radius=1, shapeName=dyn_hairFront_01_capsule, center=<0.7211 0 0.2645>, rotation=<rot 75 180 -180>}]
 dyn_hairMid_00 [{shapeType=Capsule, length=3, radius=1, shapeName=dyn_hairFront_00_capsule, center=<1.5 0 -0.6>, rotation=<rot 75 0 0>}]
 dyn_hairMid_01 [{shapeType=Capsule, length=2, radius=1, shapeName=dyn_hairFront_01_capsule, center=<0.7211 0 0.2645>, rotation=<rot 75 180 -180>}]
 dyn_hairMid_02 [{
   shapeType=Capsule
   length=2
   radius=1
   shapeName=dyn_hairFront_00_capsule
   center=<1.1119 0 0.1297>
   rotation=<rot 86.7729 -179.9994 -179.9994>
  }]
 dyn_hairMid_03 [{shapeType=Capsule, length=3.25, radius=1.75, shapeName=dyn_hairMid_03_capsule, center=<2 0 0>, rotation=<rot 90 -180 0>}]
 dyn_hairMid_04 [{
   shapeType=Capsule
   length=2.5
   radius=1
   shapeName=dyn_hairMid_04_capsule
   center=<1.0588 -0.0356 0.0633>
   rotation=<rot -84.4369 0.0011 -3.1173>
  }]
 dyn_hairMid_05 [{shapeType=Capsule, length=4, radius=1, shapeName=dyn_hairMid_04_capsule, center=<2 -0.087 0.155>, rotation=<rot -84.4369 0.0011 -3.1173>}]
 head [{shapeType=Capsule, length=7.1534, radius=6.6169, center=<4.8318 1.7277 -0.0744>, rotation=<rot -0.5801 116.4798 -88.7375>}]
}
constraints [
 {bone1Name=dyn_hairMidFront_00, bone2Name=head, swing1LimitDegrees=10, swing2LimitDegrees=10, twistLimitDegrees=2}
 {bone1Name=dyn_hairMidFront_01, bone2Name=head, swing1LimitDegrees=10, swing2LimitDegrees=10, twistLimitDegrees=2}
 {bone1Name=dyn_hairMid_05, bone2Name=head, swing1LimitDegrees=10, swing2LimitDegrees=10, twistLimitDegrees=2}
 {bone1Name=dyn_hairMid_00, bone2Name=head, swing1LimitDegrees=10, swing2LimitDegrees=10, twistLimitDegrees=2}
 {bone1Name=dyn_hairMidFront_02, bone2Name=head, swing1LimitDegrees=10, swing2LimitDegrees=10, twistLimitDegrees=2}
 {bone1Name=dyn_hairMidBack_01, bone2Name=head, swing1LimitDegrees=10, swing2LimitDegrees=10, twistLimitDegrees=2}
 {bone1Name=dyn_hairMidBack_00, bone2Name=head, swing1LimitDegrees=10, swing2LimitDegrees=10, twistLimitDegrees=2}
 {bone1Name=dyn_hairMidBack_02, bone2Name=head, swing1LimitDegrees=10, swing2LimitDegrees=10, twistLimitDegrees=2}
 {bone1Name=dyn_hairMidBack_03, bone2Name=head, swing1LimitDegrees=10, swing2LimitDegrees=10, twistLimitDegrees=2}
 {bone1Name=dyn_hairBack_00, bone2Name=head, swing1LimitDegrees=10, swing2LimitDegrees=10, twistLimitDegrees=2}
 {bone1Name=dyn_hairBack_02, bone2Name=head, swing1LimitDegrees=10, swing2LimitDegrees=10, twistLimitDegrees=2}
 {bone1Name=dyn_hairFront_00, bone2Name=head, swing1LimitDegrees=10, swing2LimitDegrees=10, twistLimitDegrees=2}
 {bone1Name=dyn_hairFront_01, bone2Name=head, swing1LimitDegrees=10, swing2LimitDegrees=10, twistLimitDegrees=2}
 {bone1Name=dyn_hairFront_02, bone2Name=head, swing1LimitDegrees=10, swing2LimitDegrees=10, twistLimitDegrees=2}
 {bone1Name=dyn_hairMid_01, bone2Name=head, swing1LimitDegrees=10, swing2LimitDegrees=10, twistLimitDegrees=2}
 {bone1Name=dyn_hairMid_02, bone2Name=head, swing1LimitDegrees=10, swing2LimitDegrees=10, twistLimitDegrees=2}
 {bone1Name=dyn_hairMid_03, bone2Name=head, swing1LimitDegrees=10, swing2LimitDegrees=10, twistLimitDegrees=2}
 {bone1Name=dyn_hairMid_04, bone2Name=head, swing1LimitDegrees=10, swing2LimitDegrees=10, twistLimitDegrees=2}
 {bone1Name=dyn_hairBack_01, bone2Name=head, swing1LimitDegrees=10, swing2LimitDegrees=10, twistLimitDegrees=2}
 {bone1Name=dyn_hairFront_04_Grape, bone2Name=head, swing1LimitDegrees=10, swing2LimitDegrees=10, twistLimitDegrees=2}
 {bone1Name=dyn_hairFront_03_Grape, bone2Name=head, swing1LimitDegrees=10, swing2LimitDegrees=10, twistLimitDegrees=2}
]
